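# Niagara cosmetic effect for the female Scholar Witch armor body, adding glowing, hazy grenades in user-set colors, magenta by default.
# Emissive_Grenades bursts two looping light-and-sprite particles once, locks them to two sampled bones or sockets on the user skeletal mesh and sine-pulses their alpha.
# Grenade_Rising_Haze continuously spawns sprites alternating between two such points by particle ID, fading in and out while computing an upward-offset RisingPosition from age.
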NiagaraSystem /BRCosmetics/Characters/Player/Female/Medium/Bodies/F_MED_Scholar_Witch_Armor/FX/NS_MED_Scholar_Witch_Armor_Body
dependencies {usedDataInterfaces [&NiagaraDataInterfaceCurve], usedDynamicInputs [], usedModules [&SystemState]}
emitters {
 Emissive_Grenades {
  emitterUpdateScript [
   &EmitterState()
   &SpawnBurst_Instantaneous(Spawn Count=2, Spawn Probability=HlslExpression() {hlslExpression="Emitter.LoopCount == 0"}, Use Spawn Probability=true)
  ]
  particleSpawnScript [
   &InitializeParticle(
    Color=$User.GrenadeLightColorAndBrightnessScale
    Lifetime=5
    Position Mode=::NewEnumerator0
    Sprite Size Mode=::NewEnumerator3
    Uniform Sprite Size=25
    Write Material Random=true
   )
   /SetVariables(
    Particles.LightExponent=2.5
    Particles.Sprite color:LinearColor=&MultiplyLinearColor(Linear Color=$User.HazeColor, Scale Factor=<1 1 1 0.8>)
   ) {isSetParameters=true, module=SetVariables_#1}
  ]
  particleUpdateScript [
   &ParticleState(
    Kill Particles When Lifetime Has Elapsed=false
    Let Infinitely Lived Particles Die When Emitter Deactivates=true
    Loop Particles Lifetime=true
   )
   &SkeletalMeshLocation(
    Bone / Socket Index=&ModuloInt(A=&ReturnExecIndex(), B=2)
    Inherit Velocity=false
    MaskSpawnGroup=false
    Mesh Sampling Type=::NewEnumerator5
    Skeletal Mesh=$User.SkeletalMesh
    Socket Sampling Mode=::NewEnumerator5
    Write Sampled Mesh Attrributes to Particle Attributes=true
   )
   &ScaleColor(
    Scale Alpha=&Lerp_Float(
     A=2
     B=3
     Alpha=&Sine(Normalized Angle=$System.LoopedAge, Period=$System.CurrentLoopDuration, Scale=0.5, Bias=0.5)
    )
    ScaleA=true
    ScaleRGB=true
   )
  ]
  rendererClasses [&NiagaraLightRendererProperties, &NiagaraSpriteRendererProperties]
  renderers [
   {rendererClass=&NiagaraLightRendererProperties, rendererIndex=0}
   {rendererClass=&NiagaraSpriteRendererProperties, rendererIndex=1}
  ]
 }
 Grenade_Rising_Haze {
  emitterUpdateScript [
   &EmitterState()
   &SpawnRate(SpawnRate=1.25, Use Spawn Probability=false)
  ]
  particleSpawnScript [
   /SetVariables(Particles.CameraOffset=3.5, Particles.Light Color:LinearColor=$User.HazeColor, Particles.LightExponent=3) {isSetParameters=true, module=SetVariables_#1}
   &InitializeParticle(
    Color=$User.HazeColor
    Lifetime Max=4
    Lifetime Min=3
    Lifetime Mode=::NewEnumerator1
    Position Mode=::NewEnumerator0
    Sprite Size Mode=::NewEnumerator3
    Uniform Sprite Size=25
    Write Material Random=true
   )
  ]
  particleUpdateScript [
   &SkeletalMeshLocation(
    Bone / Socket Index=&ModuloInt(A=&ReturnParticleIDAttributes(), B=2)
    Inherit Velocity=false
    MaskSpawnGroup=false
    Mesh Sampling Type=::NewEnumerator5
    Sampled Position Offset Coordinate Space:ENiagaraMeshTransforms=::NewEnumerator3
    Skeletal Mesh=$User.SkeletalMesh
    Socket Sampling Mode=::NewEnumerator5
    Write Sampled Mesh Attrributes to Particle Attributes=true
   ) {module=SkeletalMeshLocation001}
   &ParticleState()
   &ScaleColor(
    Scale Alpha=&FloatFromCurve(FloatCurve=DataInterface() {propertyValues {Curve {keys [<at 0 0 RCIM_Cubic>,<at 0.4 1 RCIM_Cubic>,<at 1 0 RCIM_Cubic>]}}})
    ScaleA=true
    ScaleRGB=false
   )
   /SetVariables(Particles.RisingPosition:Vector3f=&Add_Vector(
     A=Linked() {linkedVariable {name=Particles.Position, type=NiagaraPosition}}
     B=&MakeVector(
      X=0
      Y=0
      Z=&Multiply_Float(A=$Particles.NormalizedAge, B=&RandomRangeFloat(Minimum=1, Maximum=5))
     )
    )) {isSetParameters=true}
  ]
  rendererClasses [&NiagaraSpriteRendererProperties]
  renderers [{rendererClass=&NiagaraSpriteRendererProperties}]
 }
}
summary {
 systemName=NS_MED_Scholar_Witch_Armor_Body
 emitters [
  {
   emitterName=Emissive_Grenades
   rendererClasses [&NiagaraLightRendererProperties, &NiagaraSpriteRendererProperties]
  }
  {emitterName=Grenade_Rising_Haze, rendererClasses [&NiagaraSpriteRendererProperties]}
 ]
 userVariables [
  User.GrenadeLightColorAndBrightnessScale {type=LinearColor, description="", defaultValue {struct=LinearColor, value=<rgba 0.7991 0.0103 0.7991 1>}}
  User.HazeColor {type=LinearColor, description="", defaultValue {struct=LinearColor, value=<rgba 0.75 0 0.75 0.125>}}
  User.SkeletalMesh {
   type=NiagaraDataInterfaceSkeletalMesh
   defaultValue {
    struct=NiagaraExt_VariableValue_DataInterface
    value {dataInterface=&self:NiagaraDataInterfaceSkeletalMesh_0}
   }
  }
 ]
}
