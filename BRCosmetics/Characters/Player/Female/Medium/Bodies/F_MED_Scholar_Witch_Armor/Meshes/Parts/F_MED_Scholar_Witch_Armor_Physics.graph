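# Physics asset for the female medium Scholar Witch Armor cosmetic part, covering dynamic bangs, hair and hood bones around the head.
# Left, middle and right bangs each form a chain from a sphere base through two box bodies, with tight swing and twist limits.
# Hair and hood capsules on each side constrain directly to the head capsule, with the right hair allowed more swing than the left.
# No constraint links the bangs bases to the head.
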PhysicsAsset /BRCosmetics/Characters/Player/Female/Medium/Bodies/F_MED_Scholar_Witch_Armor/Meshes/Parts/F_MED_Scholar_Witch_Armor_Physics
bodies {
 dyn_bangs_01_l [{shapeType=Box, extentX=2.0815, extentY=3.1465, extentZ=0.886, center=<0.7328 0 0>}]
 dyn_bangs_01_mid [{shapeType=Box, extentX=1.8182, extentY=3.0093, extentZ=0.7773, center=<0.7462 0 0>}]
 dyn_bangs_01_r [{shapeType=Box, extentX=1.76, extentY=1.76, extentZ=0.76, center=<0.7558 0 0>}]
 dyn_bangs_base_l [{shapeType=Sphere, radius=0.5}]
 dyn_bangs_base_mid [{shapeType=Sphere, radius=0.5}]
 dyn_bangs_base_r [{shapeType=Sphere, radius=0.5}]
 dyn_bangs_parent_l [{shapeType=Box, extentX=2.26, extentY=3.01, extentZ=0.76, center=<1.0677 0 0>}]
 dyn_bangs_parent_mid [{shapeType=Box, extentX=1.8274, extentY=3.0958, extentZ=1.3303, center=<0.8789 0 0>}]
 dyn_bangs_parent_r [{shapeType=Box, extentX=1.6913, extentY=1.8157, extentZ=0.75, center=<0.9045 0 0>}]
 dyn_hair_l [{shapeType=Capsule, length=2.7896, radius=3.2594, center=<2.5336 0 1.103>, rotation=<rot -66.4723 0 0>}]
 dyn_hair_r [{shapeType=Capsule, length=2.9022, radius=2.8181, center=<2.6329 0.3413 1.5388>, rotation=<rot 11.9775 -90 59.3051>}]
 dyn_hood_l [{shapeType=Capsule, length=4.085, radius=2.6182, center=<0.8839 0 -2.4923>}]
 dyn_hood_r [{shapeType=Capsule, length=4.5683, radius=2.3423, center=<1.2239 0 -2.2803>}]
 head [{shapeType=Capsule, length=8.375, radius=8.125, center=<0.1244 0.3432 -0.0004>, rotation=<rot 89.7324 -0.0029 -8.4333>}]
}
constraints [
 {bone1Name=dyn_bangs_parent_r, bone2Name=dyn_bangs_base_r, swing1LimitDegrees=10, swing2LimitDegrees=4, twistLimitDegrees=3}
 {bone1Name=dyn_bangs_01_r, bone2Name=dyn_bangs_parent_r, swing1LimitDegrees=10, swing2LimitDegrees=8, twistLimitDegrees=3}
 {bone1Name=dyn_bangs_parent_mid, bone2Name=dyn_bangs_base_mid, swing1LimitDegrees=10, swing2LimitDegrees=4, twistLimitDegrees=3}
 {bone1Name=dyn_bangs_01_mid, bone2Name=dyn_bangs_parent_mid, swing1LimitDegrees=10, swing2LimitDegrees=8, twistLimitDegrees=3}
 {bone1Name=dyn_bangs_parent_l, bone2Name=dyn_bangs_base_l, swing1LimitDegrees=10, swing2LimitDegrees=4, twistLimitDegrees=3}
 {bone1Name=dyn_bangs_01_l, bone2Name=dyn_bangs_parent_l, swing1LimitDegrees=10, swing2LimitDegrees=8, twistLimitDegrees=3}
 {bone1Name=dyn_hair_r, bone2Name=head, swing1LimitDegrees=25, swing2LimitDegrees=20, twistLimitDegrees=5}
 {bone1Name=dyn_hair_l, bone2Name=head, swing1LimitDegrees=10, swing2LimitDegrees=10, twistLimitDegrees=3}
 {bone1Name=dyn_hood_l, bone2Name=head, swing1LimitDegrees=20, swing2LimitDegrees=20, twistLimitDegrees=3}
 {bone1Name=dyn_hood_r, bone2Name=head, swing1LimitDegrees=20, swing2LimitDegrees=20, twistLimitDegrees=3}
]
